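; Bathroom special prop: a pool that launches players into the air when enough Fortnite player pawns gather inside it.
; Box overlaps start a server-side repeating timer that counts pawns and, at the threshold, fires LaunchPlayers behind a DoOnce.
; The authority launches overlapping pawns after a warmup, then clears the trigger and calls ResetPool to re-arm.
; The remote branch plays alarm, water-level and spotlight timeline, and launch effects.
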(event EventBeginPlay
  ; Init WaterMesh Height
  (bind _output_get (Variables|Default|SetWaterMeshHighLocation (Transformation|GetWorldLocation (Variables|Default|GetSM_FoamWave1))))
  (Variables|Default|SetWaterLevelLowLocation (- _output_get "0.000000,0.000000,35.000000")))

(event OnComponentBeginOverlap(OverlapBox)
  (params (OverlappedComponent :type Primitive%20Component%20Object%20Reference) (OtherActor :type Actor%20Object%20Reference) (OtherComp :type Primitive%20Component%20Object%20Reference) (OtherBodyIndex :type Integer) (bFromSweep :type Boolean) (SweepResult :type Hit%20Result%20Structure%20%28by%20ref%29))
  ; Server Overlap, Count Players in Pool
  (Networking|SwitchHasAuthority
    (:Authority
      (if (not (Variables|Default|GetPoolTriggered))
        (bind _returnvalue (Utilities|Time|SetTimerbyEvent (delegate Custom|CheckPlayerCount) 2.0 true))
        (Variables|Default|SetHandle _returnvalue)))))

(event Custom|TestPoolLaunch
  (bind _handle (Variables|Default|GetHandle))
  ; Server Overlap, Count Players in Pool
  (Utilities|FlowControl|DoOnce false
    (:Completed
      (Variables|Default|SetPoolTriggered true)
      (Utilities|Time|PauseTimerbyHandle _handle)
      (Utilities|Time|ClearandInvalidateTimerbyHandle _handle)
      (CallFunction|LaunchPlayers))))

(event OnComponentEndOverlap(OverlapBox)
  (params (OverlappedComponent :type Primitive%20Component%20Object%20Reference) (OtherActor :type Actor%20Object%20Reference) (OtherComp :type Primitive%20Component%20Object%20Reference) (OtherBodyIndex :type Integer))
  ; Server Overlap, Count Players in Pool
  (Networking|SwitchHasAuthority
    (:Authority
      (if (not (Variables|Default|GetPoolTriggered))
        (bind _returnvalue (Utilities|Time|SetTimerbyEvent (delegate Custom|CheckPlayerCount) 2.0 true))
        (Variables|Default|SetHandle _returnvalue)))))

(event Custom|CheckPlayerCount
  (bind _handle (Variables|Default|GetHandle))
  ; Server Overlap, Count Players in Pool
  (if (> (+ (Utilities|Array|Length (Collision|GetOverlappingActors (Variables|Default|GetOverlapBox) "/Script/FortniteGame.FortPlayerPawnAthena" :!class /Script/Engine.PrimitiveComponent)) 1) (Variables|Default|GetNumOfPlayersToActivate))
    (Utilities|FlowControl|DoOnce false
      (:Completed
        (Variables|Default|SetPoolTriggered true)
        (Utilities|Time|PauseTimerbyHandle _handle)
        (Utilities|Time|ClearandInvalidateTimerbyHandle _handle)
        (CallFunction|LaunchPlayers)))
    (else
      (if (not (> (Utilities|Array|Length (Collision|GetOverlappingActors (Variables|Default|GetOverlapBox) "/Script/FortniteGame.FortPlayerPawnAthena" :!class /Script/Engine.PrimitiveComponent)) 0))
        (Utilities|Time|PauseTimerbyHandle (Variables|Default|GetHandle)) ; no players inside anymore, can pause
      ))))

(event Custom|ResetPool
  (bind _handle (Variables|Default|GetHandle))
  ; Server Overlap, Count Players in Pool
  (Utilities|FlowControl|DoOnce false
    (:Completed
      (Variables|Default|SetPoolTriggered true)
      (Utilities|Time|PauseTimerbyHandle _handle)
      (Utilities|Time|ClearandInvalidateTimerbyHandle _handle)
      (CallFunction|LaunchPlayers))))

(event Custom|LaunchPlayers
  ; LauchAllPlayers
  (Networking|SwitchHasAuthority
    (:Authority
      (Utilities|FlowControl|Delay 1.61) ; warmup time
      (for _array_element (Collision|GetOverlappingActors (Variables|Default|GetOverlapBox) "/Script/FortniteGame.FortPlayerPawnAthena" :!class /Script/Engine.PrimitiveComponent)
        (bind _asfort_pawn (Utilities|Casting|CastToFortPawn _array_element))
        (Character|LaunchCharacter _asfort_pawn (Variables|Default|GetLaunch%20Velocity) true true))
      (Utilities|FlowControl|Delay 5.0)
      (Variables|Default|SetPoolTriggered false) ; server
      (CallFunction|ResetPool) ; allow it to trigger again
    )
    (:Remote
      (Components|Activation|Deactivate (Variables|Default|GetNS_Washmachine_GoldBars))
      (Audio|Components|Audio|Play (Variables|Default|GetAlarmAudio) 0.5)
      (Utilities|FlowControl|Sequence
        (:then_0
          (bind (_direction _waterlevel _light) (Timeline|DrainPool PlayFromStart)
            (:Update
              (Transformation|SetWorldLocation (Variables|Default|GetSM_FoamWave1) (Math|Vector|Lerp(Vector) (Variables|Default|GetWaterLevelLowLocation) (Variables|Default|GetWaterMeshHighLocation) _waterlevel) false false)
              (Rendering|Components|Light|SetIntensity (Variables|Default|GetSpotLight) (Math|Float|Lerp 0.0 7.5e+02 _light)))
            (:LaunchNow
              (Components|Activation|Activate (Variables|Default|GetLaunchPlayersFX) true) ; clients
              (Audio|Components|Audio|FadeIn (Variables|Default|GetLaunchAudio) 0.125 1.0 0.0 "Linear" :!class /Script/Engine.AudioComponent))))
        (:then_1
          (Utilities|FlowControl|Delay 2.5)
          (Components|Activation|Activate (Variables|Default|GetNS_Washmachine_GoldBars) false)
          (Audio|Components|Audio|FadeOut (Variables|Default|GetAlarmAudio) 4.0 0.0 "Linear" :!class /Script/Engine.AudioComponent))))))

(fn ConstructionScript)
